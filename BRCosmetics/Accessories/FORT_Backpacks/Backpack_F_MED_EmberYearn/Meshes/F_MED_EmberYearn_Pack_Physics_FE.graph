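# Physics asset for the EmberYearn backpack cosmetic, giving collision bodies and joints to its dangling ball and cactus chains.
# Two capsules cover the pack, and base_chain attaches to it with free swing and twist.
# From dyn_chain_01 the chain forks into a ball branch ending in a sphere and a longer cactus branch ending in a larger capsule.
# Small capsules form most links, joined by swing-limited constraints with zero twist limit.
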PhysicsAsset /BRCosmetics/Accessories/FORT_Backpacks/Backpack_F_MED_EmberYearn/Meshes/F_MED_EmberYearn_Pack_Physics_FE
bodies {
 base_ball_chain [{shapeType=Capsule, length=0, radius=0.5, shapeName=base_ball_chain_capsule, rotation=<rot 0 -90 0>}]
 base_cactus_chain [{shapeType=Capsule, length=0.505, radius=0.505, shapeName=base_cactus_chain_capsule, rotation=<rot 0 -90 0>}]
 base_chain [{shapeType=Capsule, length=0, radius=0.5, shapeName=base_chain_capsule, rotation=<rot 0 -90 0>}]
 dyn_ball_chain_01 [{
   shapeType=Capsule
   length=0
   radius=0.6731
   shapeName=dyn_ball_chain_01_capsule
   center=<0.7841 0.0044 0.0522>
   rotation=<rot 0 -89.5433 -83.6407>
  }]
 dyn_ball_chain_02 [{
   shapeType=Capsule
   length=0
   radius=0.7177
   shapeName=dyn_ball_chain_02_capsule
   center=<0.7876 0.016 -0.0263>
   rotation=<rot 0 -88.6738 -88.077>
  }]
 dyn_cactus_chain_01 [{
   shapeType=Capsule
   length=0
   radius=0.6385
   shapeName=dyn_cactus_chain_01_capsule
   center=<0.7752 0.0124 0.0502>
   rotation=<rot 0 90.8942 -86.2982>
  }]
 dyn_cactus_chain_02 [{
   shapeType=Capsule
   length=0
   radius=0.7226
   shapeName=dyn_cactus_chain_02_capsule
   center=<0.7752 0.0124 0.0502>
   rotation=<rot 0 90.9434 -84.9123>
  }]
 dyn_cactus_chain_03 [{
   shapeType=Capsule
   length=0
   radius=0.6893
   shapeName=dyn_cactus_chain_03_capsule
   center=<0.7752 0.0124 0.0502>
   rotation=<rot 0 90.8942 -86.2982>
  }]
 dyn_cactus_chain_04 [{
   shapeType=Capsule
   length=0
   radius=0.7164
   shapeName=dyn_cactus_chain_04_capsule
   center=<0.7752 0.0124 0.0502>
   rotation=<rot 0 90.9434 -84.9122>
  }]
 dyn_cactus_chain_05 [{
   shapeType=Capsule
   length=1.4649
   radius=1.4695
   shapeName=dyn_cactus_chain_05_capsule
   center=<2.3311 0.0449 -0.0622>
   rotation=<rot -0.0002 91.1574 -85.9706>
  }]
 dyn_chain_01 [{
   shapeType=Capsule
   length=0.2697
   radius=1
   shapeName=dyn_chain_01_capsule
   center=<1.1871 -0.3011 -0.0256>
   rotation=<rot 0.2986 -90 -2.2127>
  }]
 dyn_chain_03 [{shapeType=Sphere, radius=1.6046, shapeName=dyn_chain_03_sphere, center=<2.1216 0.0975 -0.0385>}]
 pack [
  {shapeType=Capsule, length=15, radius=7.9669, shapeName=pack_capsule, center=<0 -10.3753 0>}
  {
   shapeType=Capsule
   length=3.0209
   radius=2.7598
   shapeName=pack_capsule
   center=<-5.8027 -7.7687 -13.3206>
   rotation=<rot -28.3509 0.0001 -0.0001>
  }
 ]
}
constraints [
 {
  bone1Name=base_chain
  bone2Name=pack
  swing1LimitDegrees=45
  swing1Motion=Free
  swing2LimitDegrees=45
  swing2Motion=Free
  twistLimitDegrees=45
  twistMotion=Free
 }
 {bone1Name=dyn_chain_01, bone2Name=base_chain, swing1LimitDegrees=20, swing2LimitDegrees=40, twistLimitDegrees=0}
 {bone1Name=base_ball_chain, bone2Name=dyn_chain_01, swing1LimitDegrees=20, swing2LimitDegrees=40, twistLimitDegrees=0}
 {bone1Name=dyn_ball_chain_01, bone2Name=base_ball_chain, swing1LimitDegrees=20, swing2LimitDegrees=40, twistLimitDegrees=0}
 {bone1Name=dyn_ball_chain_02, bone2Name=dyn_ball_chain_01, swing1LimitDegrees=30, swing2LimitDegrees=40, twistLimitDegrees=0}
 {bone1Name=dyn_chain_03, bone2Name=dyn_ball_chain_02, swing1LimitDegrees=30, swing2LimitDegrees=40, twistLimitDegrees=0}
 {bone1Name=base_cactus_chain, bone2Name=dyn_chain_01, swing1LimitDegrees=20, swing2LimitDegrees=40, twistLimitDegrees=0}
 {bone1Name=dyn_cactus_chain_01, bone2Name=base_cactus_chain, swing1LimitDegrees=20, swing2LimitDegrees=40, twistLimitDegrees=0}
 {bone1Name=dyn_cactus_chain_02, bone2Name=dyn_cactus_chain_01, swing1LimitDegrees=30, swing2LimitDegrees=40, twistLimitDegrees=0}
 {bone1Name=dyn_cactus_chain_03, bone2Name=dyn_cactus_chain_02, swing1LimitDegrees=30, swing2LimitDegrees=40, twistLimitDegrees=0}
 {bone1Name=dyn_cactus_chain_04, bone2Name=dyn_cactus_chain_03, swing1LimitDegrees=30, swing2LimitDegrees=40, twistLimitDegrees=0}
 {bone1Name=dyn_cactus_chain_05, bone2Name=dyn_cactus_chain_04, swing1LimitDegrees=30, swing2LimitDegrees=40, twistLimitDegrees=0}
]
